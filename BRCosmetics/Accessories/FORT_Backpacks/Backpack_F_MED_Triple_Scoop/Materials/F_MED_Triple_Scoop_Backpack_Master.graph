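# Master character-style material for the Triple Scoop backpack, with an optional rotator for spinning parts such as wheels.
# PreFX output passes through switchable silk, subsurface and metal LUT skin effects, masked by a glitch-function mask, into PostFX.
# The rotator spins green vertex-color areas around a Blueprint-named bone pivot, axis and angle, while red keeps standard WPO.
# Previous-frame parameters supply correct motion vectors, and rotated normals are corrected.
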Material /BRCosmetics/Accessories/FORT_Backpacks/Backpack_F_MED_Triple_Scoop/Materials/F_MED_Triple_Scoop_Backpack_Master
graph {
 properties {
  bUseMaterialAttributes=true
  bUsedWithMeshDeformer=true
  bUsedWithSkeletalMesh=true
  pixelDepthOffsetMode=PDOM_Legacy
  translucencyLightingMode=TLM_SurfacePerPixelLighting
  usedShadingModels=Default Lit
 }
 nodes [
  n1 = MaterialFunctionCall() {materialFunction=&BaseFN_CharacterFunction_PreFX}
  n2 = MaterialFunctionCall!(In=TextureObjectParameter() {group=Skin FX, parameterName=SkinFX_Mask}) {materialFunction=/Game/Weapons/WeaponSkins/MaterialFunction/MF_GlitchSubfunction}
  n3 = StaticSwitchParameter(True=@n2.Glitched, False=@n2.Glitched) {group=Skin FX, parameterName=UseGlitch}
  n4 = StaticSwitchParameter(
   True=MaterialFunctionCall!(Alpha=ChannelMaskParameter(None=@n3) {group=Skin FX, parameterName=Silk Channel}.R, Material=@n1.Result) {materialFunction=/Game/Weapons/WeaponSkins/MaterialFunction/MF_Silk}.Result
   False=@n1.Result
  ) {group=Skin FX, parameterName=Use Silk}
  n5 = StaticSwitchParameter(
   True=MaterialFunctionCall!(In=@n4, Dual SS Secondary Alpha, Alpha=StaticComponentMaskParameter(None=@n3) {defaultR=true, group=Skin FX, parameterName=SS Masks}) {materialFunction=/Game/Weapons/WeaponSkins/MaterialFunction/SS_Shader_Model}.Result
   False=@n4
  ) {group=Skin FX, parameterName="Uses Subsurface Shading Model (NOT Subsurface Profile)"}
  n6 = MaterialFunctionCall(
   AO=@n1.AO
   Distance Blend=@n1.AO
   Fresnel=@n1.AO
   Material=StaticSwitchParameter(
    True=MaterialFunctionCall(Mask=@n3, Materials=@n5) {materialFunction=&MF_MetalsLUT}.Result
    False=@n5
   ) {group=Skin FX, parameterName=UseMetalLUT}
  ) {materialFunction=&BaseFN_CharacterFunction_PostFX}
  n7 = GetMaterialAttributes(None=@n6.Result) {attributeGetTypes [#1, #2]}
  n8 = VectorParameter() {group=TRS, parameterName=BP_RightVector, defaultValue=<rgba 0 1 0 0>}
  n9 = ScalarParameter() {group=TRS, parameterName=BP_WheelAngle}
  n10 = Subtract(
   A=VectorParameter() {group=TRS, parameterName=BP_BonePos}.RGB
   B=Transform(None=VectorParameter() {group=TRS, parameterName=BoneOffset, defaultValue=<rgba 0 0 1.4 0>}.RGB) {transformSourceType=TRANSFORMSOURCE_Local}
  ) {constB=0}
  n11 = WorldPosition()
  n12 = VertexColor()
  n13 = LinearInterpolate(
   A=LinearInterpolate(A=Constant(), B=RotateAboutAxis(NormalizedRotationAxis=@n8.RGB, RotationAngle=@n9, PivotPoint=@n10, Position=@n11.XYZ), Alpha=@n12.G)
   B=@n7.World Position Offset
   Alpha=Ceil(None=@n12.R)
  )
  n14 = VectorParameter() {group=TRS, parameterName=BP_PreviousRightVector, defaultValue=<rgba 0 1 0 0>}
  n15 = ScalarParameter() {group=TRS, parameterName=BP_PreviousWheelAngle}
  n16 = Subtract(
   A=VectorParameter() {group=TRS, parameterName=BP_PreviousBonePos}.RGB
   B=Transform(None=VectorParameter() {group=TRS, parameterName=BoneOffset, defaultValue=<rgba 0 0 1.4 0>}.RGB) {transformSourceType=TRANSFORMSOURCE_Local}
  ) {constB=0}
  n17 = WorldPosition()
  n18 = VertexColor()
  n19 = LinearInterpolate(
   A=LinearInterpolate(A=Constant(), B=RotateAboutAxis(NormalizedRotationAxis=@n14.RGB, RotationAngle=@n15, PivotPoint=@n16, Position=@n17.XYZ), Alpha=@n18.G)
   B=@n7.World Position Offset
   Alpha=Ceil(None=@n18.R)
  )
  n20 = MaterialFunctionCall(Rotation Axis=@n8.RGB, Rotation Angle=@n9, World Space Vertex Normals=Transform(None=@n7.Normal)) {materialFunction=&FixRotateAboutAxisNormals}
  StaticSwitchParameter(
   True=SetMaterialAttributes!(
    MaterialAttributes=@n7.MaterialAttributes
    World Position Offset=PreviousFrameSwitch(Current Frame=@n13, Previous Frame=@n19)
    Normal=LinearInterpolate(A=@n7.Normal, B=@n20.New Tangent Space Vertex Normal, Alpha=@n12.G)
   ) {attributeSetTypes [#1, #2]}
   False=@n6.Result
  ) {group=TRS, parameterName=UseRotator}
 ]
}
